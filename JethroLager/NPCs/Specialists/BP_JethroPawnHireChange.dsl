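; Component logic that gives a hired specialist NPC its hired voice bank and voices its responses to player ping commands.
; SetBank applies the hired soundbank to the owner's NPC voice component, casts the owner to a non-participant player pawn, and binds to its convert component's ping command event.
; PingCommand plays tag-based voice lines on that pawn, using dedicated tags for Revive and Item plus a generic command-received tag.
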(event Custom|SetBank
  (bind _returnvalue (Components|GetOwner))
  (bind _output_get (Variables|Default|SetVoiceComp (Actor|GetComponentbyClass _returnvalue "/Game/Sounds/Fort_Audio_Blueprints/VoiceComponent/NPC_VoiceComponent.NPC_VoiceComponent_C")))
  (Variables|/Script/FortniteAI.FortAIComponent_Voice|SetVoiceSoundBank (Variables|Default|GetHiredSoundbank) _output_get)
  (bind _asbp_player_pawn_non_participant (Utilities|Casting|CastToBP_PlayerPawn_NonParticipant _returnvalue))
  (bind _output_get_2 (Variables|Default|SetPawn _asbp_player_pawn_non_participant))
  (ConvertEvents|BindEventtoOnPingCommandEvent (Variables|/Script/FortniteGame.FortPlayerPawn|GetConvertComponent _output_get_2) (delegate Custom|PingCommand)))

(event Custom|PingCommand
  (params (CommandType :type PingCommandType%20Enum :default "GoTo"))
  (bind _pawn (Variables|Default|GetPawn))
  (switch Utilities|FlowControl|Switch|SwitchonPingCommandType CommandType
    (:GoTo
      _)
    (:BackToMe
      _)
    (:BackToDefault
      _)
    (:HoldPosition
      _)
    (:Revive
      (Class|BPPlayerPawnNonParticipant|PlayVoiceComponentSoundbyTag _pawn "(TagName=\"Pawn.Athena.NPC.Sound.Voice.OnRevive\")"))
    (:Item
      (Class|BPPlayerPawnNonParticipant|PlayVoiceComponentSoundbyTag _pawn "(TagName=\"Pawn.Athena.NPC.Sound.Voice.GiveItem\")"))
    (:Attack
      _))
  (Class|BPPlayerPawnNonParticipant|PlayVoiceComponentSoundbyTag _pawn "(TagName=\"Pawn.Athena.NPC.Sound.Voice.CommandRecieved\")"))
